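# Sound cue for the drop sound of the BlasterPlotTwist item, classed as Fort_Loot_Collect.
# A single wave player runs through a modulator whose pitch and volume ranges are pinned, so the sound plays without random variation.
# It uses the shared weapon-drop attenuation at reduced volume and overrides concurrency to allow one instance, blocking new plays while one is active.
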SoundCue /Galileo/Items/BlasterPlotTwist/Audio/BlasterPlotTwist_Drop_Cue
root=Modulator_0
nodes [
 {id=Modulator_0, children [WavePlayer_1], properties {pitchMax=1, pitchMin=1, volumeMax=1, volumeMin=1}}
 {id=WavePlayer_1, properties {soundWaveAssetPtr=&/BlasterPlotTwist_Drop}}
]
properties {
 attenuationSettings=&Fort_Weapon_Drop_Att
 bOverrideConcurrency=true
 maxDistance=3000
 soundClassObject=&Fort_Loot_Collect
 volumeMultiplier=0.597
 concurrencyOverrides {maxCount=1, resolutionRule=PreventNew}
}
