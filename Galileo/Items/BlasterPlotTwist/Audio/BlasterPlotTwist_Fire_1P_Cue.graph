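# First-person fire sound cue for the BlasterPlotTwist weapon, played through first-person attenuation and gunfire concurrency.
# A quality-level switch sends the first two tiers to a random pick of five close-fire waves and the last two to a three-wave subset.
# A root modulator adds slight pitch variation to each shot.
# The cue uses the ranged-weapon 1P fire sound class.
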SoundCue /Galileo/Items/BlasterPlotTwist/Audio/BlasterPlotTwist_Fire_1P_Cue
root=Modulator_1
nodes [
 {id=Modulator_1, children [QualityLevel_0], properties {pitchMax=1.03, pitchMin=0.99}}
 {id=QualityLevel_0, children [Random_0,Random_0,Random_1,Random_1]}
 {id=Random_0, children [WavePlayer_9, WavePlayer_10, WavePlayer_11, WavePlayer_12, WavePlayer_13], properties {weights [1,1,1,1,1]}}
 {id=Random_1, children [WavePlayer_11, WavePlayer_12, WavePlayer_13], properties {bShouldExcludeFromBranchCulling=true, weights [1,1,1]}}
 {id=WavePlayer_10, properties {soundWaveAssetPtr=&/BlasterPlotTwist_Fire_Close_01}}
 {id=WavePlayer_11, properties {soundWaveAssetPtr=&/BlasterPlotTwist_Fire_Close_02}}
 {id=WavePlayer_12, properties {soundWaveAssetPtr=&/BlasterPlotTwist_Fire_Close_03}}
 {id=WavePlayer_13, properties {soundWaveAssetPtr=&/BlasterPlotTwist_Fire_Close_04}}
 {id=WavePlayer_9, properties {soundWaveAssetPtr=&/BlasterPlotTwist_Fire_Close_05}}
]
properties {
 attenuationSettings=&Fort_Weapon_Projectile_FirstPerson_Athena_Att
 maxDistance=2097152
 soundClassObject=&Fort_WeaponRanged_Fire_1P_PSM
 volumeMultiplier=1
 attenuationOverrides {falloffDistance=5000, attenuationShapeExtents {x=1000}}
 concurrencyOverrides {bLimitToOwner=true, maxCount=4, resolutionRule=StopOldest}
 concurrencySet [&Weapon_Gunfire_FirstPerson_Conc]
}
