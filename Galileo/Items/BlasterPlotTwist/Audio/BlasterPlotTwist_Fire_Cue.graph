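# Fire sound cue for the BlasterPlotTwist weapon, mixing close, medium and distant layers.
# Each layer has its own attenuation, modulation and sound class override: third-person fire for close, distant fire for the other two.
# Quality-level nodes choose between five-wave and three-wave random pools per layer, with the smaller pools excluded from branch culling.
# Local and global gunfire concurrency limit playback; the debug flag is on.
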SoundCue /Galileo/Items/BlasterPlotTwist/Audio/BlasterPlotTwist_Fire_Cue
root=Mixer_38
nodes [
 {
  id=Attenuation_118
  children [Modulator_0]
  properties {attenuationSettings=/Galileo/Items/BlasterPlotTwist/Audio/Settings/Fort_Weapon_BlasterPlotTwist_Close_Athena_Att}
 }
 {
  id=Attenuation_119
  children [Modulator_1]
  properties {attenuationSettings=/Galileo/Items/BlasterPlotTwist/Audio/Settings/Fort_Weapon_BlasterPlotTwist_Medium_Mono_Athena_Att}
 }
 {
  id=Attenuation_120
  children [Modulator_12]
  properties {attenuationSettings=/Galileo/Items/BlasterPlotTwist/Audio/Settings/Fort_Weapon_BlasterPlotTwist_Distant_Mono_Athena_Att}
 }
 {id=Mixer_38, children [SoundClass_77, SoundClass_76, SoundClass_0], properties {inputVolume [1,0.9,0.5]}}
 {id=Modulator_0, children [QualityLevel_3], properties {pitchMax=1, pitchMin=1}}
 {id=Modulator_1, children [QualityLevel_2], properties {pitchMax=0.98, pitchMin=0.96, volumeMax=1, volumeMin=1}}
 {id=Modulator_12, children [QualityLevel_1], properties {pitchMax=0.95, pitchMin=0.94, volumeMax=0.9, volumeMin=0.9}}
 {id=QualityLevel_1, children [Random_2,Random_2,Random_6,Random_6]}
 {id=QualityLevel_2, children [Random_1,Random_1,Random_8,Random_8]}
 {id=QualityLevel_3, children [Random_0,Random_0,Random_9,Random_9]}
 {id=Random_0, children [WavePlayer_34, WavePlayer_35, WavePlayer_37, WavePlayer_38, WavePlayer_40], properties {weights [1,1,1,1,1]}}
 {id=Random_1, children [WavePlayer_21, WavePlayer_26, WavePlayer_27, WavePlayer_31, WavePlayer_32], properties {weights [1,1,1,1,1]}}
 {id=Random_2, children [WavePlayer_1, WavePlayer_3, WavePlayer_4, WavePlayer_6, WavePlayer_20], properties {weights [1,1,1,1,1]}}
 {id=Random_6, children [WavePlayer_4, WavePlayer_6, WavePlayer_20], properties {bShouldExcludeFromBranchCulling=true, weights [1,1,1]}}
 {id=Random_8, children [WavePlayer_27, WavePlayer_31, WavePlayer_32], properties {bShouldExcludeFromBranchCulling=true, weights [1,1,1]}}
 {id=Random_9, children [WavePlayer_37, WavePlayer_38, WavePlayer_40], properties {bShouldExcludeFromBranchCulling=true, weights [1,1,1]}}
 {
  id=SoundClass_0
  children [Attenuation_120]
  properties {soundClassOverride=&Fort_WeaponRanged_Fire_Distant}
 }
 {
  id=SoundClass_76
  children [Attenuation_119]
  properties {soundClassOverride=&Fort_WeaponRanged_Fire_Distant}
 }
 {
  id=SoundClass_77
  children [Attenuation_118]
  properties {soundClassOverride=&Fort_WeaponRanged_Fire_3P_PSM}
 }
 {id=WavePlayer_1, properties {soundWaveAssetPtr=&/BlasterPlotTwist_Fire_Distant_01}}
 {id=WavePlayer_20, properties {soundWaveAssetPtr=&/BlasterPlotTwist_Fire_Distant_05}}
 {id=WavePlayer_21, properties {soundWaveAssetPtr=&/BlasterPlotTwist_Fire_Medium_01}}
 {id=WavePlayer_26, properties {soundWaveAssetPtr=&/BlasterPlotTwist_Fire_Medium_02}}
 {id=WavePlayer_27, properties {soundWaveAssetPtr=&/BlasterPlotTwist_Fire_Medium_03}}
 {id=WavePlayer_3, properties {soundWaveAssetPtr=&/BlasterPlotTwist_Fire_Distant_02}}
 {id=WavePlayer_31, properties {soundWaveAssetPtr=&/BlasterPlotTwist_Fire_Medium_04}}
 {id=WavePlayer_32, properties {soundWaveAssetPtr=&/BlasterPlotTwist_Fire_Medium_05}}
 {id=WavePlayer_34, properties {soundWaveAssetPtr=&/BlasterPlotTwist_Fire_Close_01}}
 {id=WavePlayer_35, properties {soundWaveAssetPtr=&/BlasterPlotTwist_Fire_Close_02}}
 {id=WavePlayer_37, properties {soundWaveAssetPtr=&/BlasterPlotTwist_Fire_Close_03}}
 {id=WavePlayer_38, properties {soundWaveAssetPtr=&/BlasterPlotTwist_Fire_Close_04}}
 {id=WavePlayer_4, properties {soundWaveAssetPtr=&/BlasterPlotTwist_Fire_Distant_03}}
 {id=WavePlayer_40, properties {soundWaveAssetPtr=&/BlasterPlotTwist_Fire_Close_05}}
 {id=WavePlayer_6, properties {soundWaveAssetPtr=&/BlasterPlotTwist_Fire_Distant_04}}
]
properties {
 bDebug=true
 maxDistance=27500
 volumeMultiplier=0.9
 concurrencyOverrides {resolutionRule=PreventNew}
 concurrencySet [&Weapon_Gunfire_Local_Conc, &Weapon_Gunfire_Global_Conc]
}
