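# Sound cue for the BlasterPlotTwist body impact sound, routed to the impacts sound class with bullet-impact attenuation.
# A Modulator feeds from a quality switch whose first two levels randomly pick among all six wave variations, while the last two pick among only variations 04 to 06.
# The smaller random pool is excluded from branch culling.
# A concurrency override caps simultaneous instances, stopping the oldest.
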SoundCue /Galileo/Items/BlasterPlotTwist/Audio/BlasterPlotTwist_Impact_Body_Cue
root=Modulator_1
nodes [
 {id=Modulator_1, children [QualityLevel_0]}
 {id=QualityLevel_0, children [Random_1,Random_1,Random_0,Random_0]}
 {id=Random_0, children [WavePlayer_9, WavePlayer_10, WavePlayer_11], properties {bShouldExcludeFromBranchCulling=true, weights [1,1,1]}}
 {
  id=Random_1
  children [WavePlayer_6, WavePlayer_7, WavePlayer_8, WavePlayer_9, WavePlayer_10, WavePlayer_11]
  properties {weights [1,1,1,1,1,1]}
 }
 {id=WavePlayer_10, properties {soundWaveAssetPtr=&/BlasterPlotTwist_Impact_Body_05}}
 {id=WavePlayer_11, properties {soundWaveAssetPtr=&/BlasterPlotTwist_Impact_Body_06}}
 {id=WavePlayer_6, properties {soundWaveAssetPtr=&/BlasterPlotTwist_Impact_Body_01}}
 {id=WavePlayer_7, properties {soundWaveAssetPtr=&/BlasterPlotTwist_Impact_Body_02}}
 {id=WavePlayer_8, properties {soundWaveAssetPtr=&/BlasterPlotTwist_Impact_Body_03}}
 {id=WavePlayer_9, properties {soundWaveAssetPtr=&/BlasterPlotTwist_Impact_Body_04}}
]
properties {
 attenuationSettings=/Game/Sounds/Attenuation/Fort_Bullet_Impact_Att
 bOverrideConcurrency=true
 maxDistance=5300
 soundClassObject=&Fort_Impacts
 volumeMultiplier=1.2
 concurrencyOverrides {maxCount=4, resolutionRule=StopOldest}
}
