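# Sound cue for BlasterPlotTwist projectile impacts on surfaces, in the Fort_Impacts sound class.
# A slight random delay and a modulator feed a quality-level switch: the first two levels randomly pick from six impact waves, the last two from only waves 04–06.
# It uses long bullet-impact attenuation and overrides concurrency to cap simultaneous plays, stopping the oldest.
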SoundCue /Galileo/Items/BlasterPlotTwist/Audio/BlasterPlotTwist_Impact_Surface_Cue
root=Delay_0
nodes [
 {id=Delay_0, children [Modulator_0], properties {delayMax=0.02}}
 {id=Modulator_0, children [QualityLevel_0]}
 {id=QualityLevel_0, children [Random_0,Random_0,Random_1,Random_1]}
 {
  id=Random_0
  children [WavePlayer_6, WavePlayer_7, WavePlayer_8, WavePlayer_9, WavePlayer_10, WavePlayer_11]
  properties {weights [1,1,1,1,1,1]}
 }
 {id=Random_1, children [WavePlayer_9, WavePlayer_10, WavePlayer_11], properties {bShouldExcludeFromBranchCulling=true, weights [1,1,1]}}
 {id=WavePlayer_10, properties {soundWaveAssetPtr=&/BlasterPlotTwist_Impact_Surface_05}}
 {id=WavePlayer_11, properties {soundWaveAssetPtr=&/BlasterPlotTwist_Impact_Surface_06}}
 {id=WavePlayer_6, properties {soundWaveAssetPtr=&/BlasterPlotTwist_Impact_Surface_01}}
 {id=WavePlayer_7, properties {soundWaveAssetPtr=&/BlasterPlotTwist_Impact_Surface_02}}
 {id=WavePlayer_8, properties {soundWaveAssetPtr=&/BlasterPlotTwist_Impact_Surface_03}}
 {id=WavePlayer_9, properties {soundWaveAssetPtr=&/BlasterPlotTwist_Impact_Surface_04}}
]
properties {
 attenuationSettings=/Game/Sounds/Attenuation/Fort_Bullet_ImpactOpponent_Lng_Att
 bOverrideAttenuation=true
 bOverrideConcurrency=true
 maxDistance=2600
 soundClassObject=&Fort_Impacts
 volumeMultiplier=0.9
 attenuationOverrides {falloffDistance=2200}
 concurrencyOverrides {maxCount=4, resolutionRule=StopOldest}
 concurrencySet [&RifleBulletImpact_Conc]
}
