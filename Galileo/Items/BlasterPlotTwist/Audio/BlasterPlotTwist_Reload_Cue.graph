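# Sound cue for the reload foley of the BlasterPlotTwist weapon.
# A fixed short delay node feeds a single wave player for one reload sample, with no randomization in the graph.
# It routes to the weapon foley sound class with shared reload attenuation, slightly lowered pitch and volume, a concurrency override, and the debug flag left on.
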SoundCue /Galileo/Items/BlasterPlotTwist/Audio/BlasterPlotTwist_Reload_Cue
root=Delay_0
nodes [
 {id=Delay_0, children [WavePlayer_0], properties {delayMax=0.04, delayMin=0.04}}
 {id=WavePlayer_0, properties {soundWaveAssetPtr=&/BlasterPlotTwist_Reload_05}}
]
properties {
 attenuationSettings=&Fort_Weapon_Reload_Att
 bDebug=true
 bOverrideConcurrency=true
 maxDistance=2250
 pitchMultiplier=0.95
 soundClassObject=&Fort_WeaponFoley
 volumeMultiplier=0.752
}
