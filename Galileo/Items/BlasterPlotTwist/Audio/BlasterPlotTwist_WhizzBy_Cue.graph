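# Sound cue for the projectile whizz-by of the BlasterPlotTwist ranged weapon.
# A weighted random node picks one of three waves, favoring WhizzBy_05, which has its own wider pitch modulator.
# An outer modulator randomizes pitch on every pick, so that favored wave gets stacked variation.
# Overridden attenuation turns off distance volume falloff, and the shared Athena bullet whizz-by concurrency rule governs simultaneous plays.
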SoundCue /Galileo/Items/BlasterPlotTwist/Audio/BlasterPlotTwist_WhizzBy_Cue
root=Modulator_0
nodes [
 {id=Modulator_0, children [Random_1], properties {pitchMax=1.1, pitchMin=0.9}}
 {id=Modulator_1, children [WavePlayer_0], properties {pitchMax=1.1, pitchMin=0.8}}
 {id=Random_1, children [WavePlayer_1,WavePlayer_2,Modulator_1], properties {bShouldExcludeFromBranchCulling=true, weights [1,1,2]}}
 {id=WavePlayer_0, properties {soundWaveAssetPtr=&/BlasterPlotTwist_WhizzBy_05}}
 {id=WavePlayer_1, properties {soundWaveAssetPtr=&/BlasterPlotTwist_WhizzBy_02}}
 {id=WavePlayer_2, properties {soundWaveAssetPtr=&/BlasterPlotTwist_WhizzBy_03}}
]
properties {
 bOverrideAttenuation=true
 maxDistance=2097152
 soundClassObject=&Fort_WeaponRanged
 volumeMultiplier=0.3
 attenuationOverrides {bAttenuate=false, bUseComplexCollisionForOcclusion=true, occlusionVolumeAttenuation=0, stereoSpread=5000}
 concurrencySet [/Game/Sounds/Concurrency_Rules/Athena_BulletWhizBy_Conc]
}
