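; Blaster rifle weapon Blueprint (PlotTwist) that adds audio, particle and controller feedback to the weapon overheat system.
; While heating, the authority quantizes the heat ratio into a replicated byte, whose OnRep drives the muzzle heat particles.
; Overheating triggers a second effect, a release sound (also played on reload) and force feedback.
; Firing handles a curve-scaled heat audio layer and matches the effects' first-person primitive type to the weapon mesh.
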(event EventBeginPlay
  ;[0.07485 1 0.1649] BeginPlay - Cache the instigating pawn.
  (|CallParentFunction)
  (bind _asfort_player_pawn (Utilities|Casting|CastToFortPlayerPawn (Game|GetInstigator)))
  (Variables|Default|SetFortPlayerPawn _asfort_player_pawn))

(event FortWeapon|EventOnEquippedAndAttached
  ;[0.3156 1 0.1795] On Equip -> deactivate FX or else they get stuck on if previously heated up
  (|CallParentFunction)
  (Components|Activation|Deactivate (Variables|Default|GetP_Overheat)))

(event Overheating|EventOnStartHeating
  ;[1 0.5979 0.1586] Heating Up - Set a timer to update the heating value.
  (|CallParentFunction)
  (Networking|SwitchHasAuthority
    (:Authority
      (Overheating|AssignOnOverheatValueChanged (EventDispatchers|CreateEvent :!event Custom|UpdateHeatingValue)))))

(event Custom|UpdateHeatingValue
  (params (CurrentValue :type Float%20%28single-precision%29) (MaxValue :type Float%20%28single-precision%29) (State :type EFortWeaponOverheatState%20Enum) (bIsInCooldown :type Boolean))
  ;[1 0.5979 0.1586] Heating Up - Set a timer to update the heating value.
  (Variables|Default|SetReplicatedOverheatValue (Math|Conversions|ToByte(Integer) (Math|Float|Truncate (* 255 (Math|Float|SafeDivide CurrentValue MaxValue))))))

(event Overheating|EventOnStartOverheated
  ;[1 0.3196 0] Overheated
  (|CallParentFunction)
  (Components|Activation|Activate (Variables|Default|GetP_Overheat_V2) true)
  (bind _fortplayerpawn (Variables|Default|GetFortPlayerPawn)
    (:then
      (Audio|SpawnSoundAttached "/Game/Sounds/Fort_Weapons/AutoRifle/Galileo_Bun/Galileo_Bun_Overheat_Release_Cue.Galileo_Bun_Overheat_Release_Cue" (Variables|/Script/Engine.Actor|GetRootComponent _fortplayerpawn))
      (Components|Activation|Deactivate (Variables|Default|GetP_Overheat))
      (Utilities|IsValid (Fortnite|GetFortPlayerControllerfromActor _fortplayerpawn)
        (:Is%20Valid
          (Game|Feedback|LocalPlayForceFeedback (Fortnite|GetFortPlayerControllerfromActor _fortplayerpawn) "/Game/Weapons/ForceFeedback/FF_GalileoBun_Overheated.FF_GalileoBun_Overheated" false "None"))))
    (:else)))

(event Overheating|EventOnStartCooling
  ;[0.09064 0.85 1] Start Cooling
  (|CallParentFunction)
  (Overheating|UnbindEventfromOnOverheatValueChanged (EventDispatchers|CreateEvent :!event Custom|UpdateHeatingValue)))

(event Overheating|EventOnFinishedCooling
  ;[0.0916 0.2553 1] Finished Cooling
  (|CallParentFunction)
  (Components|Activation|Deactivate (Variables|Default|GetP_Overheat)))

(event EventOnPlayWeaponFireFX
  (params (bPersistentFire :type Boolean) (bSecondaryFire :type Boolean))
  (bind _overheataudiolayer (Variables|Default|GetOverheatAudioLayer))
  ;[0.5584 0.1765 1] Weapon Fire FX - Adjust audio for heating value.
  (|CallParentFunction bPersistentFire bSecondaryFire)
  (bind _fortplayerpawn (Variables|Default|GetFortPlayerPawn)
    (:then
      (if (Pawn|IsLocallyControlled _fortplayerpawn)
        (Audio|Components|Audio|Play _overheataudiolayer 0.0)
        (Audio|Components|Audio|SetVolumeMultiplier _overheataudiolayer (Math|Curves|GetFloatValue (Variables|Default|GetOverheatAudioFeedbackCurve) (/ (Overheat|GetCurrentOverheatValue) (Overheat|GetOverheatingMaxValue))) :!class /Script/Engine.AudioComponent)
        (Rendering|SetFirstPersonPrimitiveType (Variables|Default|GetP_Overheat) (Variables|/Script/Engine.PrimitiveComponent|GetFirstPersonPrimitiveType (Variables|Visual|GetWeaponMesh)))
        (Rendering|SetFirstPersonPrimitiveType (Variables|Default|GetP_Overheat_V2) (Variables|/Script/Engine.PrimitiveComponent|GetFirstPersonPrimitiveType (Variables|Visual|GetWeaponMesh)))))
    (:else)))

(fn LocalReloadStarted
  (params (ReloadTime :type Float%20%28single-precision%29) (ReloadType :type EFortWeaponReloadType%20Enum :default "ReloadWholeClip"))
  (bind _fortplayerpawn (Variables|Default|GetFortPlayerPawn)
    (:then
      (Audio|SpawnSoundAttached "/Game/Sounds/Fort_Weapons/AutoRifle/Galileo_Bun/Galileo_Bun_Overheat_Release_Cue.Galileo_Bun_Overheat_Release_Cue" (Variables|/Script/Engine.Actor|GetRootComponent _fortplayerpawn)))
    (:else)))

(fn OnRep_ReplicatedOverheatValue
  (bind _replicatedoverheatvalue (Variables|Default|GetReplicatedOverheatValue))
  (bind _returnvalue (Math|Conversions|ToInteger(Byte) _replicatedoverheatvalue))
  (bind _returnvalue_2 (Math|Conversions|ToFloat(Integer) _returnvalue))
  (bind _returnvalue_3 (/ _returnvalue_2 255.0))
  (bind _p_overheat (Variables|Default|GetP_Overheat))
  (Networking|CanExecuteCosmeticEvents
    (:True
      (if (> _returnvalue_3 0.25)
        (Components|Activation|Activate _p_overheat true)
        (Effects|Components|ParticleSystem|SetEmitterEnable _p_overheat "MuzzleHeat" true)
        (Effects|Components|ParticleSystem|SetEmitterEnable _p_overheat "MuzzleOverheat" false)
        (Effects|Components|ParticleSystem|SetFloatParameter _p_overheat "MuzzleHeat" (Math|Float|Clamp(Float) (- (* _returnvalue_3 2.0) 1.0)))))))

(fn ConstructionScript
  (|CallParentFunction))
