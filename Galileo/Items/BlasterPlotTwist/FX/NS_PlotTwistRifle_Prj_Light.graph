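# Niagara system for the PlotTwist rifle projectile's light, with a single PureLight emitter that renders only through a light renderer.
# A burst spawns a blue-tinted light particle slightly offset from the system location.
# Its color scales by a quick-rise, slow-decay curve and a time-of-day scalar that dims it at night.
# System.NearCamFadeLerp, apparently derived from a clamped distance remap, fades both brightness and light radius.
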NiagaraSystem /Galileo/Items/BlasterPlotTwist/FX/NS_PlotTwistRifle_Prj_Light
dependencies {
 usedDataInterfaces [&NiagaraDataInterfaceCurve]
 usedDynamicInputs [&ClampFloat, &FN_LOD_Distance_Remap]
 usedModules [
  &self:SystemSpawnScript.SystemScriptSource.SystemScriptGraph.NiagaraNodeAssignment_0.SetVariables_#1
  &self:SystemSpawnScript.SystemScriptSource.SystemScriptGraph.NiagaraNodeAssignment_1.SetVariables_#2
  &SystemState
 ]
}
emitters {PureLight {
  emitterUpdateScript [
   &EmitterState(Life Cycle Mode=::NewEnumerator1, Loop Behavior=::NewEnumerator1, Loop Duration Mode=::NewEnumerator1, UseLoopDelay=false)
   &SpawnBurst_Instantaneous(Use Spawn Probability=false, UseLoopCountLimit=false)
  ]
  particleSpawnScript [
   &InitializeParticle(Color=<rgba 0.5 0.875 5 1>, UsePositionOffset=false, Write Material Random=true)
   &SystemLocation(MaskSpawnGroup=false, Offset=<-10 0 0>)
  ]
  particleUpdateScript [
   &ParticleState(Kill Particles When Lifetime Has Elapsed=false, Let Infinitely Lived Particles Die When Emitter Deactivates=true)
   &ScaleColor(
    Scale Mode=::NewEnumerator1
    Scale RGBA=&MakeVector4_FromFloat(Float=&Multiply_Float(
      A=&Multiply_Float(
       A=&FNTimeOfDayScalar(Night Scale=0.4, Morning Scale=1, Evening Scale=1)
       B=&FloatFromCurve(FloatCurve=DataInterface() {propertyValues {Curve {keys [
            {arriveTangent=9.0725, interpMode=RCIM_Cubic, leaveTangent=9.0725, tangentMode=RCTM_User, time=0, value=0}
            {arriveTangent=-2.5249, interpMode=RCIM_Cubic, leaveTangent=-2.5249, tangentMode=RCTM_User, time=0.1408, value=0.9847}
            {arriveTangent=-0.0405, interpMode=RCIM_Cubic, leaveTangent=-0.0405, tangentMode=RCTM_User, time=1, value=0}
           ]}}})
      )
      B=$System.NearCamFadeLerp
     ))
    ScaleRGBA=true
   )
   /SetVariables(Particles.LightRadius=&Lerp_Float(A=0, B=70, Alpha=$System.NearCamFadeLerp)) {isSetParameters=true}
  ]
  rendererClasses [&NiagaraLightRendererProperties]
  renderers [{rendererClass=&NiagaraLightRendererProperties, rendererIndex=0}]
 }}
summary {
 systemName=NS_PlotTwistRifle_Prj_Light
 emitters [{emitterName=PureLight, rendererClasses [&NiagaraLightRendererProperties]}]
 userVariables []
}
