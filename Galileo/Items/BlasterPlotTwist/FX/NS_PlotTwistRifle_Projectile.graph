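# Niagara projectile effect for the PlotTwist rifle, drawing a bright cyan sprite and mesh between the muzzle and the owning component.
# The only enabled emitter, Projectile_Mesh, eases position from User.MuzzlePosition toward the owner by particle age, orienting and stretching the mesh by their separation.
# Alpha fades in shortly after spawn, while dynamic material parameters receive muzzle position and an age curve.
# Smoke ribbon, debug and empty emitters are disabled.
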NiagaraSystem /Galileo/Items/BlasterPlotTwist/FX/NS_PlotTwistRifle_Projectile
dependencies {
 usedDataInterfaces [&NiagaraDataInterfaceCurve, &NiagaraDataInterfaceColorCurve]
 usedDynamicInputs [&MakeFloatFromVector]
 usedModules [
  &SystemState
  &self:SystemSpawnScript.SystemScriptSource.SystemScriptGraph.NiagaraNodeAssignment_0.SetVariables_#1
 ]
}
emitters {
 Empty {
  bEnabled=false
  emitterUpdateScript [
   &EmitterState()
   &SpawnBurst_Instantaneous(Use Spawn Probability=false, UseLoopCountLimit=false)
  ]
  particleSpawnScript [&InitializeParticle(
    Color Mode=::NewEnumerator0
    Lifetime=5
    Sprite Size Mode=::NewEnumerator3
    Uniform Sprite Size=50
    UsePositionOffset=false
    Write Material Random=true
   )]
  particleUpdateScript [&ParticleState()]
  rendererClasses [&NiagaraSpriteRendererProperties]
  renderers [{rendererClass=&NiagaraSpriteRendererProperties}]
 }
 FN_Debug {
  bEnabled=false
  emitterUpdateScript [
   &EmitterState(Life Cycle Mode=::NewEnumerator1, Loop Behavior=::NewEnumerator1, Loop Delay=0.1, UseLoopDelay=false)
   &SpawnBurst_Instantaneous(Spawn Count=1, Use Spawn Probability=false)
  ]
  particleSpawnScript [/Niagara/Modules/Spawn/Initialization/InitializeParticle(
    Lifetime=5
    Module.Write Color=true
    Module.Write Lifetime=true
    Module.Write Mass=true
    Module.Write Material Random=true
    Module.Write Position=true
    Module.Write Scale=false
    Module.Write SpriteRotation=false
    Module.Write SpriteSize=true
    Position=&Add_Vector(A=$User.MuzzlePosition, B=<0 0 50>)
    Sprite Size=<100 100>
   )]
  particleUpdateScript [
   &ParticleState()
   /SetVariables(Particles.Debug:NiagaraFloat=$User.ScanLength) {isSetParameters=true}
   &DynamicMaterialParameters(
    Index 0 Param 1=$Particles.Debug
    Module.Param0WriteEnabled=true
    Module.Param1WriteEnabled=true
    Module.Param2WriteEnabled=true
    Module.Param3WriteEnabled=true
   )
  ]
  rendererClasses [&NiagaraSpriteRendererProperties]
  renderers [{rendererClass=&NiagaraSpriteRendererProperties}]
 }
 Projectile_Mesh {
  emitterUpdateScript [
   &EmitterState(Life Cycle Mode=::NewEnumerator1, Loop Behavior=::NewEnumerator1, Loop Duration Mode=::NewEnumerator1, UseLoopDelay=false)
   &SpawnBurst_Instantaneous(Use Spawn Probability=false, UseLoopCountLimit=false)
  ]
  particleSpawnScript [
   /Niagara/Modules/Spawn/Initialization/V2/InitializeParticle(
    Color=<rgba 0 35.1565 50 1>
    Mesh Scale Mode=::NewEnumerator4
    Sprite Size Mode=::NewEnumerator3
    Uniform Sprite Size=5
    UsePositionOffset=false
    Write Material Random=true
   )
   &SystemLocation(MaskSpawnGroup=false, Offset Coordinate Space=::World)
   &InitializeParticle(
    Color Mode=::NewEnumerator0
    Lifetime=&Lerp_Float(
     A=0.1
     B=1
     Alpha=&RemapRange(Input Value To Remap=$User.ScanLength, Input Min=0, Input Max=1000, Output Min=0, Output Max=1)
    )
    Mesh Scale Mode=::NewEnumerator4
    Sprite Size Mode=::NewEnumerator3
    UsePositionOffset=false
    Write Material Random=true
   ) {enabled=false, module=InitializeParticle001}
  ]
  particleUpdateScript [
   &ParticleState(Kill Particles When Lifetime Has Elapsed=false, Let Infinitely Lived Particles Die When Emitter Deactivates=true)
   &DynamicMaterialParameters(
    Index 0 Param 1=$System.Muzzle_Pos_X
    Index 0 Param 2=$System.Muzzle_Pos_Y
    Index 0 Param 3=$System.Muzzle_Pos_Z
    Index 0 Param 4=&FloatFromCurve(FloatCurve=DataInterface() {propertyValues {Curve {keys [<at 0 0>,<at 0.1 1>,<at 1 10>]}}}, CurveIndex=$Particles.Age, Scale Curve=1000)
    Module.Param0WriteEnabled=true
    Module.Param1WriteEnabled=true
    Module.Param2WriteEnabled=true
    Module.Param3WriteEnabled=true
   )
   &ScaleMeshSize(Scale Factor=&MakeVector(
     X=&ClampFloat(
      Float=&Divide_Float(
       A=&VectorLength(Vector=&Subtract_Vector(A=$User.MuzzlePosition, B=Linked() {linkedVariable {name=Engine.Owner.Position, type=NiagaraPosition}}))
       B=100
      )
      Min=0.1
      Max=10
     )
     Y=1
     Z=1
    ))
   &UpdateMeshOrientation(
    Facing Direction=&NormalizeVector(
     Vector To Normalize=&Subtract_Vector(A=Linked() {linkedVariable {name=Engine.Owner.Position, type=NiagaraPosition}}, B=$User.MuzzlePosition)
     Safe Normalize=true
     Safe Fallback Threshold=0
     Safe Fallback Vector=<0 0 0.001>
    )
    Facing Interpolation=false
    Orientation Method=::NewEnumerator1
    Override Quat=false
    TransformFacing=false
    TransformUp=false
   )
   /SetVariables(Particles.Position=&Lerp_Vector(
     A=$User.MuzzlePosition
     B=Linked() {linkedVariable {name=Engine.Owner.Position, type=NiagaraPosition}}
     Alpha=&Lerp_Float(
      A=&FloatFromCurve(
       FloatCurve=DataInterface() {propertyValues {Curve {keys [<at 0 0>, {arriveTangent=15.1596, leaveTangent=15.1596, time=0.1, value=1}]}}}
       CurveIndex=$Particles.Age
      )
      B=&FloatFromCurve(
       FloatCurve=DataInterface() {propertyValues {Curve {keys [{arriveTangent=0.7022, time=0, value=0}, {arriveTangent=5.581, leaveTangent=5.581, time=1, value=1}]}}}
       CurveIndex=$Particles.Age
      )
      Alpha=&RemapRange(
       Input Value To Remap=&VectorLength(Vector=&Subtract_Vector(A=$User.MuzzlePosition, B=Linked() {linkedVariable {name=Engine.Owner.Position, type=NiagaraPosition}}))
       Input Min=100
       Input Max=15000
       Output Min=0
       Output Max=1
      )
     )
    )) {enabled=false, isSetParameters=true}
   /SetVariables(Particles.Position=&Lerp_Vector(
     A=$User.MuzzlePosition
     B=Linked() {linkedVariable {name=Engine.Owner.Position, type=NiagaraPosition}}
     Alpha=&FloatFromCurve(
      FloatCurve=DataInterface() {propertyValues {Curve {keys [
          {arriveTangent=2.265, interpMode=RCIM_Cubic, leaveTangent=2.265, tangentMode=RCTM_User, time=0, value=0}
          {arriveTangent=0.2891, interpMode=RCIM_Cubic, leaveTangent=0.2891, tangentMode=RCTM_User, time=1, value=1}
         ]}}}
      CurveIndex=$Particles.Age
     )
    )) {isSetParameters=true}
   &ScaleColor(
    Scale Alpha=&FloatFromCurve(FloatCurve=DataInterface() {propertyValues {Curve {keys [<at 0.15 0>,<at 0.25 1>]}}}, CurveIndex=$Particles.Age)
    ScaleA=true
    ScaleRGB=true
   )
  ]
  rendererClasses [&NiagaraSpriteRendererProperties, &NiagaraMeshRendererProperties]
  renderers [
   {rendererClass=&NiagaraSpriteRendererProperties}
   {rendererClass=&NiagaraMeshRendererProperties, rendererIndex=1}
  ]
 }
 Smoke_Trail {
  bEnabled=false
  emitterUpdateScript [
   &EmitterState(Life Cycle Mode=::NewEnumerator1, Loop Duration=1, UseLoopDelay=false)
   &SpawnPerUnit(
    Max Movement Threshold=100000
    Module.Use Max Movement Threshold=true
    Module.Use Movement Tolerance=true
    Movement Tolerance=0.1
    Spawn Spacing=50
    Use Spawn Probability=false
   )
  ]
  particleSpawnScript [
   &InitializeRibbon(Lifetime=0.2, Ribbon Width=10, UsePositionOffset=false, Write Material Random=true)
   &SystemLocation(MaskSpawnGroup=false, Offset=<100 0 0>)
  ]
  particleUpdateScript [
   &AccelerationForce(Acceleration=<0 0 30>)
   &SolveForcesAndVelocity(Clamp Velocity=false, Limit Acceleration=false)
   &ParticleState()
   &ScaleRibbonWidth(Ribbon Width Scale=&FloatFromCurve(FloatCurve=DataInterface() {propertyValues {Curve {keys [<at 0 1>,<at 0.5 2>]}}}))
   &ScaleColor(
    Scale Alpha=&FloatFromCurve(FloatCurve=DataInterface() {propertyValues {Curve {keys [<at 0 1>,<at 1 0>]}}}, Scale Curve=3)
    Scale RGB=&MakeVectorFromLinearColorRGB(Color=&ColorFromCurve(ColorCurve=DataInterface() {propertyValues {
        AlphaCurve {keys []}
        BlueCurve {keys [{arriveTangent=1.3532, interpMode=RCIM_Cubic, time=0, value=1}, {interpMode=RCIM_Cubic, leaveTangent=1.3532, time=0.0627, value=0.2969}]}
        GreenCurve {keys [{arriveTangent=0.3631, interpMode=RCIM_Cubic, time=0, value=0.7031}, {interpMode=RCIM_Cubic, leaveTangent=0.3631, time=0.0627, value=0.2969}]}
        RedCurve {keys [{arriveTangent=-3.205, interpMode=RCIM_Cubic, time=0, value=0}, {interpMode=RCIM_Cubic, leaveTangent=-3.205, time=0.0627, value=0.2969}]}
       }}))
    ScaleA=true
    ScaleRGB=true
   )
   &DynamicMaterialParameters(
    Index 0 Param 1=&FloatFromCurve(FloatCurve=DataInterface() {propertyValues {Curve {keys [<at 0 1>,<at 0.4 0.56>,<at 1 0>]}}})
    Index 0 Param 2=&FloatFromCurve(FloatCurve=DataInterface() {propertyValues {Curve {keys [<at 0 2>,<at 0.15 0>]}}})
    Index 0 Param 3=1
    Index 0 Param 4=1
    Module.Param0WriteEnabled=true
    Module.Param1WriteEnabled=true
    Module.Param2WriteEnabled=true
    Module.Param3WriteEnabled=true
   )
  ]
  rendererClasses [&NiagaraRibbonRendererProperties]
  renderers [{rendererClass=&NiagaraRibbonRendererProperties}]
 }
}
summary {
 systemName=NS_PlotTwistRifle_Projectile
 emitters [
  {
   emitterName=Projectile_Mesh
   rendererClasses [&NiagaraSpriteRendererProperties, &NiagaraMeshRendererProperties]
  }
  {bEnabled=false, emitterName=Empty, rendererClasses [&NiagaraSpriteRendererProperties]}
  {bEnabled=false, emitterName=Smoke_Trail, rendererClasses [&NiagaraRibbonRendererProperties]}
  {bEnabled=false, emitterName=FN_Debug, rendererClasses [&NiagaraSpriteRendererProperties]}
 ]
 userVariables [
  User.ActorPosition {type=Vector3f, description="", defaultValue {struct=Vector3f, value=<0 0 0>}}
  User.ForwardVector {type=Vector3f, description="", defaultValue {struct=Vector3f, value=<0 0 0>}}
  User.MuzzlePosition {type=Vector3f, description="", defaultValue {struct=Vector3f, value=<0 0 0>}}
  User.ScanLength {type=NiagaraFloat, defaultValue=0, description=""}
  User.StartTime {type=NiagaraFloat, defaultValue=0, description=""}
 ]
}
