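# Niagara muzzle flash for the PlotTwist blaster: a brief blue burst lasting about a tenth of a second.
# MFlashHot and MFlashHot001 each spawn one forward-drifting mesh with random roll, stretched then narrowed by matching scale curves.
# MFlashHot animates brightness with an RGB curve, while MFlashHot001 is a smaller mesh with a constant HDR color.
# Glow_Light adds a sprite and particle light, writing a forward-offset light position attribute.
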NiagaraSystem /Galileo/Items/BlasterPlotTwist/FX/NS_PlotTwist_MF_01
dependencies {
 usedDataInterfaces [
  &NiagaraDataInterfaceVectorCurve
  &NiagaraDataInterfaceMeshRendererInfo
  &NiagaraDataInterfaceCurve
 ]
 usedDynamicInputs []
 usedModules [&SystemState]
}
emitters {
 Glow_Light {
  emitterUpdateScript [
   &EmitterState()
   &SpawnBurst_Instantaneous(Use Spawn Probability=false, UseLoopCountLimit=false) {module=SpawnBurst_Instantaneous001}
  ]
  particleSpawnScript [
   /Niagara/Modules/Spawn/Initialization/InitializeParticle(
    Color=<rgba 0.1 0.175 1 1>
    Lifetime=&UniformRangedFloat(Minimum=0.1, Maximum=0.12)
    Module.Write Color=true
    Module.Write Lifetime=true
    Module.Write Mass=false
    Module.Write Material Random=true
    Module.Write Position=true
    Module.Write Scale=false
    Module.Write SpriteRotation=false
    Module.Write SpriteSize=true
    Sprite Rotation=&RandomRangeFloat(Minimum=0, Maximum=360)
    Sprite Size=<180 15>
   )
   /Game/Effects/Niagara/Modules/Orientation/FN_Fan(Spin:NiagaraFloat=&UniformRangedFloat(Minimum=0, Maximum=1))
   /SetVariables(Particles.LightPositionNew:Vector3f=&Add_Vector(
     A=Linked() {linkedVariable {name=Particles.Position, type=NiagaraPosition}}
     B=/Niagara/DynamicInputs/Transforms/TransformPosition(Position=<25 0 0>, Source Space=::Local, Destination Space=::Simulation)
    )) {isSetParameters=true}
  ]
  particleUpdateScript [
   &ParticleState()
   &ScaleMeshSize(Scale Factor=&VectorFromFloat(Value=&FloatFromCurve(FloatCurve=DataInterface() {propertyValues {Curve {keys [{arriveTangent=2.2457, interpMode=RCIM_Cubic, leaveTangent=2.2457, tangentMode=RCTM_User, time=0, value=0}, <at 1 1>]}}})))
   &DynamicMaterialParameters(Write Parameter Index 0=false)
   &FN_Light_Attributes(
    Color=<rgba 1.4 1.8667 7 1>
    Create Color Attribute=true
    Create Exponent Attribute=true
    Create Radius Attribute=true
    Enable Particle Light Enable=true
    Exponent=3
    Light Enabled=true
    Radius=250
   )
  ]
  rendererClasses [&NiagaraLightRendererProperties, &NiagaraSpriteRendererProperties]
  renderers [
   {rendererClass=&NiagaraLightRendererProperties, rendererIndex=0}
   {rendererClass=&NiagaraSpriteRendererProperties, rendererIndex=1}
  ]
 }
 MFlashHot {
  emitterUpdateScript [
   &EmitterState()
   &SpawnBurst_Instantaneous(Spawn Count=1, Use Spawn Probability=false)
  ]
  particleSpawnScript [
   /Niagara/Modules/Spawn/Initialization/V2/InitializeParticle(
    Lifetime=0.12
    Mesh Scale=<1 1.5 1.5>
    Mesh Scale Mode=::NewEnumerator4
    Position Offset=<-10 0 0>
    UsePositionOffset=true
    Write Material Random=true
   ) {module=InitializeParticle001}
   /SetVariables(Particles.MeshOrientation=&MultiplyQuaternion(
     Quaternion A=&MakeQuatFromAxisAngle(
      Axis=&RandomRangeVector(Minimum=<-1 0 0>, Maximum=<1 0 0>, Evaluation Type=::NewEnumerator1)
      AngleInDegrees=90
     )
     Quaternion B=<0 0 0 1>
    )) {isSetParameters=true}
   &Color(Color=&MakeLinearColorFromVectorAndFloat(
     "Vector (RGB)"=&MakeVectorFromLinearColorRGB(Color=<rgba 0.05 0.1292 1 0>)
     "Float (Alpha)"=&ClampFloat(Float=1)
    ))
   &AddVelocity(Velocity=<50 0 0>)
  ]
  particleUpdateScript [
   &SolveForcesAndVelocity(Clamp Velocity=false, Limit Acceleration=false)
   &ParticleState()
   &DynamicMaterialParameters(Module.Param0WriteEnabled=true, Module.Param1WriteEnabled=true, Module.Param2WriteEnabled=true, Module.Param3WriteEnabled=true)
   &ScaleColor(
    Scale RGB=&VectorFromCurve(VectorCurve=DataInterface() {propertyValues {
       XCurve {keys [<at 0 7>,<at 0.1 1>,<at 0.7 3>]}
       YCurve {keys [{arriveTangent=-1.0128, time=0, value=7}, {arriveTangent=0.3157, time=0.1, value=1}, <at 0.7 3>]}
       ZCurve {keys [{arriveTangent=-6.017, time=0, value=7.017}, {arriveTangent=2, leaveTangent=2, time=0.1, value=1}, <at 0.6996 3>]}
      }})
    ScaleA=true
    ScaleRGB=true
   )
   &ScaleMeshSize(Scale Factor=&VectorFromCurve(VectorCurve=DataInterface() {propertyValues {
       XCurve {keys [
         {arriveTangent=3.0736, interpMode=RCIM_Cubic, leaveTangent=3.0736, tangentMode=RCTM_User, time=0, value=0.2}
         {arriveTangent=0.178, interpMode=RCIM_Cubic, leaveTangent=0.178, tangentMode=RCTM_User, time=0.3, value=1}
         {arriveTangent=-0.1748, interpMode=RCIM_Cubic, leaveTangent=-0.1748, tangentMode=RCTM_User, time=1, value=0.5}
        ]}
       YCurve {keys [
         {arriveTangent=0.6326, interpMode=RCIM_Cubic, leaveTangent=0.6326, tangentMode=RCTM_User, time=0, value=0.55}
         {arriveTangent=-0.1563, interpMode=RCIM_Cubic, leaveTangent=-0.1563, tangentMode=RCTM_User, time=0.3, value=0.65}
         {arriveTangent=-0.5198, interpMode=RCIM_Cubic, leaveTangent=-0.5198, tangentMode=RCTM_User, time=1, value=0}
        ]}
       ZCurve {keys [
         {arriveTangent=0.6326, interpMode=RCIM_Cubic, leaveTangent=0.6326, tangentMode=RCTM_User, time=0, value=0.55}
         {arriveTangent=-0.1563, interpMode=RCIM_Cubic, leaveTangent=-0.1563, tangentMode=RCTM_User, time=0.3, value=0.65}
         {arriveTangent=-0.5198, interpMode=RCIM_Cubic, leaveTangent=-0.5198, tangentMode=RCTM_User, time=1, value=0}
        ]}
      }}))
  ]
  rendererClasses [&NiagaraMeshRendererProperties]
  renderers [{rendererClass=&NiagaraMeshRendererProperties, rendererIndex=0}]
 }
 MFlashHot001 {
  emitterUpdateScript [
   &EmitterState()
   &SpawnBurst_Instantaneous(Spawn Count=1, Use Spawn Probability=false)
  ]
  particleSpawnScript [
   &InitializeParticle(
    Color=<rgba 4.5 11.25 45 1>
    Lifetime=0.12
    Mesh Scale=<0.65 0.55 0.55>
    Mesh Scale Mode=::NewEnumerator4
    Position Offset=<-10 0 0>
    UsePositionOffset=true
    Write Material Random=true
   ) {module=InitializeParticle001}
   /SetVariables(Particles.MeshOrientation=&MultiplyQuaternion(
     Quaternion A=&MakeQuatFromAxisAngle(
      Axis=&RandomRangeVector(Minimum=<-1 0 0>, Maximum=<1 0 0>, Evaluation Type=::NewEnumerator1)
      AngleInDegrees=90
     )
     Quaternion B=<0 0 0 1>
    )) {isSetParameters=true}
   &AddVelocity(Velocity=<50 0 0>)
  ]
  particleUpdateScript [
   &SolveForcesAndVelocity(Clamp Velocity=false, Limit Acceleration=false)
   &ParticleState()
   &DynamicMaterialParameters(Module.Param0WriteEnabled=true, Module.Param1WriteEnabled=true, Module.Param2WriteEnabled=true, Module.Param3WriteEnabled=true)
   &ScaleMeshSize(Scale Factor=&VectorFromCurve(VectorCurve=DataInterface() {propertyValues {
       XCurve {keys [
         {arriveTangent=3.0736, interpMode=RCIM_Cubic, leaveTangent=3.0736, tangentMode=RCTM_User, time=0, value=0.2}
         {arriveTangent=0.178, interpMode=RCIM_Cubic, leaveTangent=0.178, tangentMode=RCTM_User, time=0.3, value=1}
         {arriveTangent=-0.1748, interpMode=RCIM_Cubic, leaveTangent=-0.1748, tangentMode=RCTM_User, time=1, value=0.5}
        ]}
       YCurve {keys [
         {arriveTangent=0.6326, interpMode=RCIM_Cubic, leaveTangent=0.6326, tangentMode=RCTM_User, time=0, value=0.55}
         {arriveTangent=-0.1563, interpMode=RCIM_Cubic, leaveTangent=-0.1563, tangentMode=RCTM_User, time=0.3, value=0.65}
         {arriveTangent=-0.5198, interpMode=RCIM_Cubic, leaveTangent=-0.5198, tangentMode=RCTM_User, time=1, value=0}
        ]}
       ZCurve {keys [
         {arriveTangent=0.6326, interpMode=RCIM_Cubic, leaveTangent=0.6326, tangentMode=RCTM_User, time=0, value=0.55}
         {arriveTangent=-0.1563, interpMode=RCIM_Cubic, leaveTangent=-0.1563, tangentMode=RCTM_User, time=0.3, value=0.65}
         {arriveTangent=-0.5198, interpMode=RCIM_Cubic, leaveTangent=-0.5198, tangentMode=RCTM_User, time=1, value=0}
        ]}
      }}))
  ]
  rendererClasses [&NiagaraMeshRendererProperties]
  renderers [{rendererClass=&NiagaraMeshRendererProperties, rendererIndex=0}]
 }
}
summary {
 systemName=NS_PlotTwist_MF_01
 emitters [
  {emitterName=MFlashHot, rendererClasses [&NiagaraMeshRendererProperties]}
  {emitterName=MFlashHot001, rendererClasses [&NiagaraMeshRendererProperties]}
  {emitterName=Glow_Light, rendererClasses [&NiagaraLightRendererProperties, &NiagaraSpriteRendererProperties]}
 ]
 userVariables []
}
