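# Physics asset for the PlotTwist rifle mesh of the BlasterPlotTwist item in Galileo.
# It holds a single body on the Handle bone, shaped as one capsule that is offset from the bone and rotated to lie along the weapon.
# It has no constraints, so the rifle has one rigid collision shape with no separately simulated or jointed parts.
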PhysicsAsset /Galileo/Items/BlasterPlotTwist/Meshes/PlotTwist_Rifle_PhysicsAsset
bodies {Handle [{shapeType=Capsule, length=59.0678, radius=10.3213, center=<30.997 1.1125 0.9009>, rotation=<rot 0 90.3439 -87.1216>}]}
constraints []
